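# Season 18 Fortnitemares conversation database in which a Dark Jonesy NPC delivers a two-path prophecy for oracle quest q05.
# The entry leads to an auto-selected main-quest "Prophecy" choice, gated on that NPC's tag, an active quest and an uncompleted objective.
# A linear chain of speech lines advances through Continue choices into a task that completes the objective by backend name.
# That task carries an inverted quest-to-turn-in requirement.
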ConversationDatabase /BattlePassS18/Conversations/Shared_Conversation_S18_Fortnitemares
entry_points [{tag=Conversation.Entry.Season18.Fortnitemares, destinations [#1]}]
nodes [
 ConversationEntryPointNode_7 {
  id=#1
  class=ConversationEntryPointNode
  outputs [#2]
  properties {entryTag {tagName=Conversation.Entry.Season18.Fortnitemares}}
 }
 Speech_C_1 {
  id=#2
  class=Speech_C
  outputs [#3]
  properties {
   softTaskIcon=&T_UI_NPC_MainQuest
   generalConfig {defaultMessage="My vision is now clear, but I bring <b>dark tidings.</>"}
  }
  sub_nodes [
   ConversationChoiceNode_20 {
    class=ConversationChoiceNode
    properties {
     defaultChoiceDisplayText=Prophecy
     choiceTags {gameplayTags [{tagName=Conversation.Choice.Category.MainQuestline}, {tagName=Conversation.Choice.AutoSelectIfValid}]}
    }
   }
   FortConversationRequirement_ControllerRequirement_40 {
    class=FortConversationRequirement_ControllerRequirement
    properties {
     nodeName=Require: Has Completed Quest
     requirement {
      class=FortControllerRequirement_HasCompletedQuest
      properties {bInverseCondition=true, questDefinition=/BattlePassS18/Items/QuestItems/Transient/quest_s18_transient_darkjonesy_oracle_q05}
     }
    }
   }
   FortConversationRequirement_ControllerRequirement_41 {
    class=FortConversationRequirement_ControllerRequirement
    properties {
     nodeName=Require: Is Quest Active
     requirement {
      class=FortControllerRequirement_IsQuestActive
      properties {questDefinition=&quest_s18_transient_darkjonesy_oracle_q05}
     }
    }
   }
   FortConversationRequirement_ControllerRequirement_42 {
    class=FortConversationRequirement_ControllerRequirement
    properties {
     nodeName=Require: Has Completed Quest Objective
     requirement {
      class=FortControllerRequirement_HasCompletedQuestObjective
      properties {
       bInverseCondition=true
       objectiveBackendName=quest_s18_transient_darkjonesy_oracle_q05_obj0
       questDefinition=&quest_s18_transient_darkjonesy_oracle_q05
      }
     }
    }
   }
   FortConversationRequirement_ControllerRequirement_58 {
    class=FortConversationRequirement_ControllerRequirement
    properties {
     nodeName=Require: NPC Has Conversation Tag
     participantId {tagName=Conversation.Participant.NPC}
     requirement {class=FortControllerRequirement_NPCHasConversationEntryTag, properties {queryTag {tagName=Conversation.Entry.DarkJonesy}}}
    }
   }
  ]
 }
 Speech_C_7 {
  id=#3
  class=Speech_C
  outputs [#4]
  properties {generalConfig {defaultMessage="<b>Two paths</> lie ahead of you."}}
  sub_nodes [ConversationChoiceNode_3 {class=ConversationChoiceNode, properties {defaultChoiceDisplayText=Continue}}]
 }
 Speech_C_0 {
  id=#4
  class=Speech_C
  outputs [#5]
  properties {generalConfig {defaultMessage="On one path, I see <b>sacrifice and loss</>."}}
  sub_nodes [ConversationChoiceNode_6 {class=ConversationChoiceNode, properties {defaultChoiceDisplayText=Continue}}]
 }
 Speech_C_4 {
  id=#5
  class=Speech_C
  outputs [#6]
  properties {generalConfig {defaultMessage="On the other, <enemy>total annihilation</>."}}
  sub_nodes [ConversationChoiceNode_23 {class=ConversationChoiceNode, properties {defaultChoiceDisplayText=Continue}}]
 }
 FortConversationTaskNode_CompleteQuestObjectiveByBackendName_0 {
  id=#6
  class=FortConversationTaskNode_CompleteQuestObjectiveByBackendName
  properties {
   objectiveBackendName=quest_s18_transient_darkjonesy_oracle_q05_obj0
   questDefinition=&quest_s18_transient_darkjonesy_oracle_q05
  }
  sub_nodes [FortConversationRequirement_ControllerRequirement_66 {
    class=FortConversationRequirement_ControllerRequirement
    properties {
     nodeName=Require: Has Quest To Turn In
     requirement {
      class=FortControllerRequirement_HasQuestToTurnIn
      properties {bInverseCondition=true, softQuestToTurnIn=&quest_s18_transient_darkjonesy_oracle_q05}
     }
    }
   }]
 }
]
